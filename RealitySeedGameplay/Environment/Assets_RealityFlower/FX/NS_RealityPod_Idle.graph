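# Niagara idle effect for the Reality Pod, pairing a looping, spinning Energy mesh with GPU Embers sprites.
# A system Timeline drives shared surges: mesh spin, flickering brightness and vertical stretch, plus ember upward acceleration.
# Embers spawn on a sphere-based shape, then drift with drag, curl noise and an orbit offset written to a separate position attribute.
# Energy's camera offset varies by platform set; a User.Hit bool is exposed, with no visible use here.
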NiagaraSystem /RealitySeedGameplay/Environment/Assets_RealityFlower/FX/NS_RealityPod_Idle
dependencies {
 usedDataInterfaces [&NiagaraDataInterfaceCurve, &NiagaraDataInterfacePlatformSet]
 usedDynamicInputs [&NormalizeVector]
 usedModules [&SystemState, &Timeline]
}
emitters {
 Embers {
  simTarget=GPUComputeSim
  emitterUpdateScript [
   &EmitterState()
   &SpawnRate(Spawn Probability=0.5, SpawnRate=30, Use Spawn Probability=true)
  ]
  particleSpawnScript [
   &InitializeParticle(
    Color Maximum=<rgba 0.0269 0.25 0.974 1>
    Color Minimum=<rgba 1 0.1978 0.7007 1>
    Color Mode=::NewEnumerator2
    Lifetime Max=3
    Lifetime Mode=::NewEnumerator1
    Sprite Size Mode=::NewEnumerator1
    Uniform Sprite Size Max=15
    UsePositionOffset=false
    Write Material Random=true
   ) {module=InitializeParticle001}
   &ShapeLocation(
    Hemisphere Distribution=<1 0.025>
    MaskSpawnGroup=true
    Offset Mode=::NewEnumerator0
    Override Seed=false
    Rotation Mode=::NewEnumerator2
    Sphere Radius=50
    Sphere Surface Distribution=1
    UseOwnerScale=true
    Yaw / Pitch / Roll=<0 0 0.025>
   )
   &AddVelocity(
    Module.Use Falloff Distance=false
    Origin Offset=<0 10 100>
    Override Seed=false
    Velocity Mode=::NewEnumerator1
    Velocity Speed=&FloatFromCurve(
     FloatCurve=DataInterface() {propertyValues {Curve {keys [<at 0.25 5 RCIM_Cubic>, {interpMode=RCIM_Cubic, leaveTangent=-228.4063, time=0.75, value=100}]}}}
     CurveIndex=&RandomRangeFloat(Minimum=0, Maximum=1)
     Scale Curve=-1
    )
   )
   &AddVelocity(
    Velocity=$Particles.ShapeLocation.ShapeBitangent
    Velocity Speed Scale=&RandomRangeFloat(Minimum=-5, Maximum=5)
   ) {module=AddVelocity001}
  ]
  particleUpdateScript [
   &ParticleState()
   &Drag(Drag=&RandomRangeFloat(Minimum=0.5, Maximum=1), Use Linear Drag=true, Use Rotational Drag=false)
   &AccelerationForce(
    Acceleration=&Lerp_Vector(
     A=<0 0 25>
     B=<0 0 500>
     Alpha=&FloatFromCurve(
      FloatCurve=DataInterface() {propertyValues {Curve {keys [
          {arriveTangent=15.6865, interpMode=RCIM_Cubic, leaveTangent=15.6865, tangentMode=RCTM_User, time=0, value=0}
          <at 0.1408 0.9847 RCIM_Cubic>
          {arriveTangent=-0.0405, interpMode=RCIM_Cubic, leaveTangent=-0.0405, tangentMode=RCTM_User, time=0.5, value=0}
         ]}}}
      CurveIndex=$System.Timeline.NormalizedTime
     )
    )
    Coordinate Space=::Local
   )
   /Niagara/Modules/Update/Forces/CurlNoiseForce(
    Noise Frequency=1
    Noise Quality / Cost=::NewEnumerator0
    Noise Strength=&FloatFromCurve(
     FloatCurve=DataInterface() {propertyValues {Curve {keys [{interpMode=RCIM_Cubic, tangentMode=RCTM_User, time=0, value=1}, {interpMode=RCIM_Cubic, tangentMode=RCTM_User, time=1, value=-1}]}}}
     Scale Curve=150
    )
    Pan Noise=false
   )
   &RotateAroundPoint(
    Radius:NiagaraFloat=&FloatFromCurve(
     FloatCurve=DataInterface() {propertyValues {Curve {keys [{interpMode=RCIM_Cubic, tangentMode=RCTM_User, time=0, value=0.25}, {interpMode=RCIM_Cubic, tangentMode=RCTM_User, time=1, value=1}]}}}
     Scale Curve=6
    )
    Rotation Rate:NiagaraFloat=&RandomRangeFloat(Minimum=0.5, Maximum=1)
    Set Particle Position=false
   )
   &SolveForcesAndVelocity(Clamp Velocity=false, Limit Acceleration=false)
   /SetVariables(Particles.Rotational_Position:NiagaraPosition=&Add_Vector(
     A=Linked() {linkedVariable {name=Particles.Position, type=NiagaraPosition}}
     B=Linked() {linkedVariable {name=Output.RotateAroundPoint.OutputPosition, type=NiagaraPosition}}
    )) {isSetParameters=true}
   &ScaleColor(
    Scale Alpha=&FloatFromCurve(
     FloatCurve=DataInterface() {propertyValues {Curve {keys [
         {arriveTangent=26.7273, interpMode=RCIM_Cubic, leaveTangent=26.7273, tangentMode=RCTM_User, time=0, value=0}
         {interpMode=RCIM_Cubic, tangentMode=RCTM_User, time=0.0967, value=0.9767}
         {interpMode=RCIM_Cubic, tangentMode=RCTM_User, time=0.66, value=1}
         <at 1 0 RCIM_Cubic>
        ]}}}
     Scale Curve=&RandomRangeFloat(Minimum=0.5, Maximum=2)
    )
    ScaleA=true
    ScaleRGB=true
   )
   &ScaleSpriteSize(Scale Factor=&Vector2DFromFloat(Value=&FloatFromCurve(FloatCurve=DataInterface() {propertyValues {Curve {keys [<at -0.0046 0>,<at 0.025 1>,<at 0.66 1>,<at 1.0011 0.002>]}}})))
  ]
  rendererClasses [&NiagaraSpriteRendererProperties]
  renderers [{rendererClass=&NiagaraSpriteRendererProperties}]
 }
 Energy {
  emitterUpdateScript [
   &EmitterState(
    Inactive Response=::NewEnumerator1
    Life Cycle Mode=::NewEnumerator1
    Loop Behavior=::NewEnumerator1
    Loop Duration Mode=::NewEnumerator1
    UseLoopDelay=false
   )
   &SpawnBurst_Instantaneous(Spawn Group=2, Use Spawn Probability=false, UseLoopCountLimit=false) {module=SpawnBurst_Instantaneous002}
  ]
  particleSpawnScript [
   &InitializeParticle(
    Color=<rgba 1 1 1 0.25>
    Lifetime=2
    Mesh Scale=<1 1 0.75>
    Mesh Scale Mode=::NewEnumerator4
    UsePositionOffset=false
    Write Material Random=true
   )
   &InitialMeshOrientation(
    Mesh Axis To Orient=<0 0 1>
    Mesh Orientation Mode=::NewEnumerator2
    Rotation=&RandomRangeVector(Minimum=<0 0 0>, Maximum=<0 0 1>)
    Use Rotation Vector=true
   )
   &DynamicMaterialParameters(
    Index 0 Param 2=&RandomRangeFloat(Minimum=0, Maximum=1.5)
    Index 0 Param 3=6
    Index 2 Param 2=20
    Index 2 Param 3=1
    Index 2 Param 4=0.25
    Module.Param0WriteEnabled=true
    Module.Param0WriteEnabled2=false
    Module.Param1WriteEnabled=true
    Module.Param1WriteEnabled2=true
    Module.Param2WriteEnabled=true
    Module.Param2WriteEnabled2=true
    Module.Param3WriteEnabled=true
    Module.Param3WriteEnabled2=true
    Write Parameter Index 2=true
   )
  ]
  particleUpdateScript [
   &ParticleState(
    Kill Particles When Lifetime Has Elapsed=false
    Let Infinitely Lived Particles Die When Emitter Deactivates=true
    Loop Particles Lifetime=true
   )
   &UpdateMeshOrientation(
    Facing Coordinate Space:ENiagaraCoordinateSpace=::Local
    Facing Direction=<0 0.1 1>
    Facing Interpolation=false
    Orientation Method=::NewEnumerator1
    Override Quat=false
    TransformFacing=true
    TransformUp=false
   ) {module=UpdateMeshOrientation002}
   &UpdateMeshOrientation(
    Rotation Rate=&Lerp_Float(
     A=-0.25
     B=-1
     Alpha=&FloatFromCurve(
      FloatCurve=DataInterface() {propertyValues {Curve {keys [{arriveTangent=11.5837, time=0, value=0}, <at 0.1 1>, <at 0.33 1>, {tangentMode=RCTM_User, time=1, value=0}]}}}
      CurveIndex=$System.Timeline.NormalizedTime
     )
    )
    Rotation Vector=<0 0 1>
   )
   &ScaleColor(
    Scale RGB=&Multiply_VectorByFloat(
     Vector=&Lerp_Vector(
      A=<4 4 4>
      B=<16 16 16>
      Alpha=&FloatFromCurve(
       FloatCurve=DataInterface() {propertyValues {Curve {keys [
           {arriveTangent=1.2413, interpMode=RCIM_Cubic, time=0, value=0}
           {arriveTangent=3.6434, interpMode=RCIM_Cubic, leaveTangent=-2.1892, tangentMode=RCTM_Break, time=0.33, value=0.9847}
           {interpMode=RCIM_Cubic, leaveTangent=-2.2403, time=1, value=0}
          ]}}}
       CurveIndex=$System.Timeline.NormalizedTime
      )
     )
     Float=&Waveform(Global Amplitude Scale=0.75, "Waveform [1]"=::NewEnumerator12, "[1] Frequency"=2, "[1] Amplitude Min/Max"=<0.6 1>)
    )
    ScaleA=true
    ScaleRGB=true
   )
   &CameraOffset(Camera Offset Amount=&MakeCustomFloatFromBool(
     Bool=&IsPlatformSetActive(New Platform Set=DataInterface() {propertyValues {Platforms {
         qualityLevelMask=-32
         deviceProfileStates [
          {profileName=Mobile, qualityLevelMask=4294967295, setQualityLevelMask=4294967295}
          {profileName=Switch, qualityLevelMask=4294967295, setQualityLevelMask=4294967295}
         ]
        }}})
     True Float=0
     False Float=15
    ))
   &ScaleMeshSize(Scale Factor=&Lerp_Vector(
     A=<1 1 1>
     B=<0.8 0.8 1.25>
     Alpha=&FloatFromCurve(
      FloatCurve=DataInterface() {propertyValues {Curve {keys [
          {arriveTangent=1.2413, interpMode=RCIM_Cubic, time=0, value=0}
          {arriveTangent=3.6434, interpMode=RCIM_Cubic, leaveTangent=-2.1892, tangentMode=RCTM_Break, time=0.33, value=0.9847}
          {interpMode=RCIM_Cubic, leaveTangent=-2.2403, time=1, value=0}
         ]}}}
      CurveIndex=$System.Timeline.NormalizedTime
     )
    ))
  ]
  rendererClasses [&NiagaraMeshRendererProperties]
  renderers [{rendererClass=&NiagaraMeshRendererProperties, rendererIndex=0}]
 }
}
summary {
 systemName=NS_RealityPod_Idle
 emitters [
  {emitterName=Energy, rendererClasses [&NiagaraMeshRendererProperties]}
  {emitterName=Embers, simTarget=GPUComputeSim, rendererClasses [&NiagaraSpriteRendererProperties]}
 ]
 userVariables [User.Hit {type=NiagaraBool}]
}
