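# Material function that drives the growth of RealitySapling fruit by deforming vertices through World Position Offset.
# A pre-skinned height gradient, scaled by FruitGrowth and ShrinkScale, blends in an offset between each vertex and PivotOffset, fading to zero at full growth.
# It adds this to incoming WPO, clamps Z, reuses X for Y and transforms the result from local space.
# Disconnected nodes hold a world-space variant with sine debugging.
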MaterialFunction /RealitySeedGameplay/items/RealitySaplingFruits/Assets/Materials/MFC_FruitGrowth
graph {nodes [
  n1 = GetMaterialAttributes(None=FunctionInput() {inputName=Material, inputType=FunctionInput_MaterialAttributes}) {attributeGetTypes [#1]}
  n2 = VectorParameter() {parameterName=PivotOffset, defaultValue=<rgba 0 0 18 1>}
  n3 = Subtract(
   A=Add(A=@n2.RGB, B=Constant3Vector())
   B=MaterialFunctionCall() {materialFunction=&LocalPosition}.Local Position
  )
  n4 = PreSkinnedPosition()
  n5 = VectorParameter() {parameterName=GradientPivotOffset, defaultValue=<rgba 0 0 2 1>}
  n6 = Divide(A=Add(A=OneMinus(None=@n4), B=Add(A=@n5.RGB, B=@n2.RGB) {constB=0}), B=PreSkinnedLocalBounds().Max)
  n7 = VertexInterpolator(VS=@n6)
  n8 = ScalarParameter() {defaultValue=1, parameterName=FruitGrowth, sliderMax=1}
  n9 = Multiply(
   A=ComponentMask(None=ShaderStageSwitch(PixelShader=@n7.PS, VertexShader=@n6)) {b=true}
   B=LinearInterpolate(A=ScalarParameter() {defaultValue=-0.3, parameterName=ShrinkScale}, Alpha=Clamp(None=@n8)) {constA=-0.2, constB=0}
  )
  n10 = MaterialFunctionCall(Float3=Add(A=@n1.WorldPositionOffset, B=LinearInterpolate(B=@n3, Alpha=@n9) {constB=0})) {materialFunction=&BreakOutFloat3Components}
  n11 = MaterialFunctionCall(
   X=@n10.R
   Y=@n10.R
   Z=Clamp(
    None=@n10.B
    Max=MaterialFunctionCall(Float3=@n3) {materialFunction=&BreakOutFloat3Components}.B
   ) {maxDefault=9999999827968, minDefault=-10000000272564224}
  ) {materialFunction=&MakeFloat3}
  FunctionOutput(None=SetMaterialAttributes!(MaterialAttributes=@n1.MaterialAttributes, WorldPositionOffset=Transform(None=@n11.Result) {transformSourceType=TRANSFORMSOURCE_Local}) {attributeSetTypes [#1]})
  n12 = Add(
   A=VectorParameter() {defaultValue=<rgba 0 0 18 1>}.RGB
   B=MaterialFunctionCall() {materialFunction=&ObjectPivotPoint}.Object Pivot Location
  )
  n13 = WorldPosition()
  n14 = PreSkinnedLocalBounds()
  n15 = Divide(A=PreSkinnedPosition(), B=Subtract(A=@n14.Max, B=@n14.Max))
  n16 = ShaderStageSwitch(PixelShader=VertexInterpolator(VS=@n15).PS, VertexShader=@n15)
  n17 = OneMinus(None=ConstantBiasScale(None=ComponentMask(None=@n16) {b=true}))
  LinearInterpolate(
   A=LinearInterpolate(B=Subtract(A=@n12, B=@n13.XYZ), Alpha=@n17)
   Alpha=MaterialFunctionCall() {materialFunction=&DebugTimeSine}.Result
  )
  MaterialFunctionCall() {materialFunction=&DebugTimeSine}
  ObjectPositionWS()
  ScalarParameter() {defaultValue=1, parameterName=FruitGrowth, sliderMax=1}
  Step(Y=@n17) {constX=0.5}
 ]}
