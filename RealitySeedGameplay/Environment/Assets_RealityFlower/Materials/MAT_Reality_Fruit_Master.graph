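# Master material for Reality Fruit meshes in the RealityFlower environment set, built on the shared character material function pipeline.
# CharacterSkinFX builds material attributes from the PreFX base, then a static switch, on by default, routes them through MFC_FruitGrowth.
# A slow time-driven sine, scaled by WindWPOStrength, rotates local position about a tilted axis to add wind sway to World Position Offset.
# Character and cinematic PostFX functions finish the result; the PivotOffset parameter is left unconnected.
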Material /RealitySeedGameplay/Environment/Assets_RealityFlower/Materials/MAT_Reality_Fruit_Master
graph {
 properties {
  bAutomaticallySetUsageInEditor=false
  bUseMaterialAttributes=true
  bUsedWithClothing=true
  bUsedWithMeshDeformer=true
  bUsedWithMorphTargets=true
  bUsedWithNanite=true
  bUsedWithSkeletalMesh=true
  pixelDepthOffsetMode=PDOM_Legacy
  refractionMethod=RM_IndexOfRefraction
  shadingModel=MSM_FromMaterialExpression
  subsurfaceProfile=&SS_HeroSkin_02
  usedShadingModels=Default Lit | Subsurface | Clear Coat | Subsurface Profile | Cloth | Eye
 }
 nodes [
  n1 = Time()
  n2 = MaterialFunctionCall() {materialFunction=&LocalPosition}
  n3 = MaterialFunctionCall(Float3=@n2.Local Position) {materialFunction=&BreakOutFloat3Components}
  n4 = Add(A=@n3.R, B=@n3.R)
  Add(A=@n1, B=Frac(None=Add(A=@n4, B=@n3.B)))
  Constant()
  n5 = MaterialFunctionCall() {materialFunction=&BaseFN_CharacterFunction_PreFX}
  n6 = MaterialFunctionCall(
   PreskinnedPosition=@n5.Pre skinned Local Position
   Gmap=@n5.Pre skinned Local Position
   MaterialAttributes=@n5.Pre skinned Local Position
   AO=@n5.Pre skinned Local Position
  ) {materialFunction=&CharacterSkinFX}
  n7 = GetMaterialAttributes(None=StaticSwitchParameter(
    True=MaterialFunctionCall!(Material=@n6.Result) {materialFunction=/RealitySeedGameplay/items/RealitySaplingFruits/Assets/Materials/MFC_FruitGrowth}.Result
    False=@n6.Result
   ) {defaultValue=true, parameterName=Use FruitGrowth}) {attributeGetTypes [#1]}
  n8 = Sine(None=Multiply(A=@n1) {constB=0.3})
  n9 = ScalarParameter() {defaultValue=0.002, parameterName=WindWPOStrength}
  n10 = Constant3Vector()
  n11 = MaterialFunctionCall() {materialFunction=&ObjectScale}
  n12 = Transform(None=RotateAboutAxis(
    NormalizedRotationAxis=Constant3Vector() {constant=<rgba 1 0.1 0 0>}
    RotationAngle=Multiply(A=@n8, B=@n9)
    PivotPoint=Multiply(A=@n10, B=@n11.Scale XYZ)
    Position=@n2.Local Position
   )) {transformSourceType=TRANSFORMSOURCE_Local}
  n13 = MaterialFunctionCall(
   AO=@n5.AO
   Distance Blend=@n5.AO
   Fresnel=@n5.AO
   Gmap=@n5.AO
   Pre Skinned Local Position=@n5.AO
   Material=SetMaterialAttributes!(MaterialAttributes=@n7.MaterialAttributes, World Position Offset=Add(A=@n7.World Position Offset, B=@n12)) {attributeSetTypes [#1]}
  ) {materialFunction=&BaseFN_CharacterFunction_PostFX}
  MaterialFunctionCall(In=@n13.Result) {materialFunction=&CINE_MF_CharPostFX}
  VectorParameter() {parameterName=PivotOffset, defaultValue=<rgba 0 0 18 1>}
 ]
}
